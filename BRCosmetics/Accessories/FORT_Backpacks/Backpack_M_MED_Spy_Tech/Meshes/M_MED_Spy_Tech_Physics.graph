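# Physics asset for the M_MED_Spy_Tech backpack cosmetic, giving its side pouches and tubes limited physical movement.
# A small capsule on the pack bone acts as the hub, with box bodies for the left and right pouches and capsules for two tubes.
# Each of the four dyn_ bodies is constrained directly to the pack with tight swing and twist limits, restricting it to slight wobble.
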PhysicsAsset /BRCosmetics/Accessories/FORT_Backpacks/Backpack_M_MED_Spy_Tech/Meshes/M_MED_Spy_Tech_Physics
bodies {
 dyn_pouch_l [{shapeType=Box, extentX=3.8148, extentY=8.0579, extentZ=11.7385, center=<0.1449 -1.7563 0.265>, rotation=<rot -0.4352 1.17 -87.5279>}]
 dyn_pouch_r [{shapeType=Box, extentX=3.8148, extentY=8.0579, extentZ=11.7385, center=<0.1513 1.9896 0.1028>, rotation=<rot 0 -8.8088 -87.4902>}]
 dyn_tubeA [{shapeType=Capsule, length=11.0405, radius=0.8988, center=<3.1015 -0.1913 0.041>, rotation=<rot 0 -89.9513 -89.4159>}]
 dyn_tubeB [{shapeType=Capsule, length=11.0214, radius=0.9706, center=<2.9734 -0.1537 0.1>, rotation=<rot 0 85.1792 -89.9972>}]
 pack [{shapeType=Capsule, length=1.2, radius=1.2, rotation=<rot 0 90 0>}]
}
constraints [
 {bone1Name=dyn_pouch_r, bone2Name=pack, swing1LimitDegrees=3, swing2LimitDegrees=7, twistLimitDegrees=7}
 {bone1Name=dyn_pouch_l, bone2Name=pack, swing1LimitDegrees=2, swing2LimitDegrees=7, twistLimitDegrees=7}
 {bone1Name=dyn_tubeB, bone2Name=pack, swing1LimitDegrees=8, swing2LimitDegrees=6, twistLimitDegrees=6}
 {bone1Name=dyn_tubeA, bone2Name=pack, swing1LimitDegrees=8, swing2LimitDegrees=6, twistLimitDegrees=6}
]
